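# Masked, lit Niagara ribbon material for the chest energy effect in the Kaiju demon defeat FX.
# Panning cloud noise, distorted by perlin curl and time-offset by a dynamic Rand parameter, drives normal-based displacement, a derived normal and a pink-to-charcoal color blend.
# The mask combines that noise with inverted Fresnel, temporal dither, a fade along U and particle alpha.
# A PreviousFrameSwitch adds a particle-velocity offset to the previous frame's displacement, apparently for motion blur.
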Material /GrillMoon/Assets/FX/Kaiju/Fight/Demon/Defeat/M_ChestEnergy_KaijuDemon
graph {
 properties {
  bAutomaticallySetUsageInEditor=false
  bUsedWithNiagaraRibbons=true
  blendMode=BLEND_Masked
  translucencyLightingMode=TLM_VolumetricPerVertexDirectional
  usedShadingModels=Default Lit
 }
 nodes [
  n1 = Multiply(A=TextureCoordinate(), B=ComponentMask(None=VectorParameter() {parameterName=Param_1, defaultValue=<rgba 4 1 0 0>}.RGB) {g=true, r=true})
  n2 = Multiply(
   A=ScalarParameter() {defaultValue=0.1, parameterName=Time}
   B=Add(A=Time(), B=DynamicParameter() {paramNames [Rand,Param2,Param3,Param4]}.Rand)
  )
  n3 = TextureSample(UVs=Multiply(A=ScalarParameter() {defaultValue=0.1, parameterName=Param_4}, B=TextureCoordinate())) {samplerType=SAMPLERTYPE_LinearColor, texture=/Game/Accessories/FORT_Backpacks/Backpack_M_MED_Carabus/FX/T_PerlinCurl}
  n4 = TextureSample(UVs=Add(A=Panner(Coordinate=@n1, Time=@n2) {speedX=-1}, B=ComponentMask(None=@n3.RGB) {g=true, r=true})) {texture=&AtmosphericCloudNoise01}
  n5 = VertexNormalWS()
  n6 = Multiply(A=Multiply(A=ParticleDirection()) {constB=-1}, B=ParticleSpeed())
  n7 = ScalarParameter() {defaultValue=0.5}
  Add(
   A=Multiply(A=Multiply(A=@n4.G, B=@n5), B=ScalarParameter() {defaultValue=1, parameterName=Param_2})
   B=PreviousFrameSwitch(Current Frame=Constant(), Previous Frame=Multiply(A=@n6, B=@n7))
  )
  Constant()
  Constant() {r=1}
  n8 = TextureCoordinate()
  DeriveNormalZ(InXY=Add(A=ConstantBiasScale(None=@n8) {bias=-0.5, scale=2}, B=@n4.G))
  LinearInterpolate(
   A=VectorParameter() {parameterName=Param_2, defaultValue=<rgba 2 0 0.4335 1>}.RGB
   B=VectorParameter() {defaultValue=<rgba 0.035 0.035 0.035 0>}.RGB
   Alpha=Power(Base=@n4.G, Exp=ScalarParameter() {defaultValue=0.15, parameterName=Param_5}) {constExponent=0.25}
  )
  n9 = OneMinus(None=ComponentMask(None=TextureCoordinate()) {r=true})
  n10 = Fresnel(ExponentIn=ScalarParameter() {defaultValue=2, parameterName=Param_1}) {exponent=1}
  n11 = Power(Base=Multiply(A=@n4.G, B=OneMinus(None=@n10)), Exp=ScalarParameter() {defaultValue=0.5})
  n12 = LinearInterpolate(
   A=MaterialFunctionCall(Alpha Threshold=@n11) {materialFunction=&DitherTemporalAA}.Result
   B=ScalarParameter() {defaultValue=2, parameterName=Param_3}
   Alpha=@n11
  )
  LinearInterpolate(B=ScalarParameter() {defaultValue=1, parameterName=TipFade}, Alpha=Multiply(A=ParticleColor().A, B=Multiply(A=@n9, B=@n12)))
  Normalize()
 ]
}
