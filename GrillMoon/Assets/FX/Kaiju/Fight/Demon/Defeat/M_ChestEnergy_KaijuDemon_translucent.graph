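# Translucent lit material for the chest energy effect in the Kaiju Demon defeat FX, enabled for Niagara ribbons.
# A panning atmospheric cloud noise feeds a derived normal, a dark color blend and a vertex-normal offset that billows the surface.
# Opacity combines the noise with an inverted Fresnel, a fade along U, particle alpha and a TipFade parameter for soft edges.
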Material /GrillMoon/Assets/FX/Kaiju/Fight/Demon/Defeat/M_ChestEnergy_KaijuDemon_translucent
graph {
 properties {
  bAutomaticallySetUsageInEditor=false
  bUsedWithNiagaraRibbons=true
  blendMode=BLEND_Translucent
  translucencyLightingMode=TLM_VolumetricPerVertexDirectional
  usedShadingModels=Default Lit
 }
 nodes [
  Constant()
  Constant() {r=1}
  n1 = TextureCoordinate()
  n2 = TextureSample(UVs=Panner() {speedX=-1}) {texture=&AtmosphericCloudNoise01}
  DeriveNormalZ(InXY=Add(A=ConstantBiasScale(None=@n1) {bias=-0.5, scale=2}, B=@n2.G))
  LinearInterpolate(B=VectorParameter() {defaultValue=<rgba 0.035 0.035 0.035 0>}.RGB, Alpha=@n2.G)
  n3 = Fresnel(ExponentIn=ScalarParameter() {defaultValue=0.9, parameterName=Param_1}) {exponent=1}
  n4 = Power(Base=Multiply(A=@n2.G, B=OneMinus(None=@n3)), Exp=ScalarParameter() {defaultValue=0.1})
  n5 = OneMinus(None=ComponentMask(None=TextureCoordinate()) {r=true})
  LinearInterpolate(B=ScalarParameter() {defaultValue=1, parameterName=TipFade}, Alpha=Multiply(A=ParticleColor().A, B=Multiply(A=@n4, B=@n5)))
  Multiply(A=Multiply(A=@n2.G, B=VertexNormalWS()), B=ScalarParameter() {defaultValue=15, parameterName=Param_2})
 ]
}
