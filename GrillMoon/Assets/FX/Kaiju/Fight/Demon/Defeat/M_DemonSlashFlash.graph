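# Unlit, two-sided translucent Niagara material for a slash flash effect in the Demon kaiju's defeat FX.
# A sphere mask forms the shape, with UVs offset in Y by a generated band and warped by a panning normal texture.
# The mask, reshaped by power and contrast, indexes a curve-atlas ramp that tints particle color.
# A powered mask, particle alpha and optional curve alpha combine into an opacity value.
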Material /GrillMoon/Assets/FX/Kaiju/Fight/Demon/Defeat/M_DemonSlashFlash
graph {
 properties {
  bAutomaticallySetUsageInEditor=false
  bUsedWithNiagaraMeshParticles=true
  bUsedWithNiagaraRibbons=true
  bUsedWithNiagaraSprites=true
  blendMode=BLEND_Translucent
  previewMesh=&/SM_GrillMoon_FanCards
  shadingModel=MSM_Unlit
  twoSided=true
  usedShadingModels=Unlit
 }
 nodes [
  Add(A=MaterialFunctionCall() {materialFunction=&LinearGradient}.VGradient)
  n1 = ParticleColor()
  n2 = MaterialFunctionCall(Float2=TextureCoordinate()) {materialFunction=&BreakOutFloat2Components}
  n3 = Constant() {r=0.5}
  n4 = Constant()
  n5 = StaticBool() {value=true}
  n6 = Add(
   A=@n2.G
   B=LinearInterpolate(
    A=ScalarParameter() {defaultValue=0.25, parameterName=Y side offset}
    Alpha=MaterialFunctionCall!(Width=@n3, Sharpness=@n4, Offset, Direction Switch=@n5, Compare, Input Coordinates) {materialFunction=/Engine/Functions/Engine_MaterialFunctions03/Procedurals/GeneratedBand}.Result
   ) {constB=0}
  )
  n7 = VectorParameter() {parameterName=Distortn Tile Speed, channelNames {a=S Y, b=S X, g=T Y, r=T X}, defaultValue=<rgba 1 0.2 0 0.25>}
  n8 = Panner(
   Coordinate=Multiply(A=TextureCoordinate() {coordinateIndex=1}, B=AppendVector(A=@n7.T X, B=@n7.T X))
   Speed=AppendVector(A=@n7.S X, B=@n7.S X)
  ) {bFractionalPart=true}
  n9 = ComponentMask(None=TextureSample(UVs=@n8) {samplerType=SAMPLERTYPE_Normal, texture=&/Substance_graph_normal}.RGB) {g=true, r=true}
  n10 = ScalarParameter() {defaultValue=0.35, parameterName=Distort Str}
  n11 = SphereMask(
   A=Add(
    A=MaterialFunctionCall(X=@n2.R, Y=@n6) {materialFunction=&MakeFloat2}.Result
    B=Multiply(A=@n9, B=@n10) {constB=0.35}
   )
   B=Constant() {r=0.5}
   Radius=ScalarParameter() {defaultValue=0.4, parameterName=Sphere mask Radius}
   Hardness=ScalarParameter() {parameterName=Sphere mask hardness}
  )
  n12 = VectorParameter() {parameterName=Color Curve U Options, channelNames {a=contrast, b=add, g=mul, r=pow}, defaultValue=<rgba 2.5 1 0 0>}
  n13 = Power(Base=@n11, Exp=@n12.pow) {constExponent=2.5}
  n14 = MaterialFunctionCall(In=Add(A=Multiply(A=@n13, B=@n12.mul), B=@n12.add), Contrast=@n12.contrast) {materialFunction=&CheapContrast}
  n15 = CurveAtlasRowParameter(CurveTime=LinearInterpolate(
    A=ScalarParameter() {parameterName=Curve reamp A}
    B=ScalarParameter() {defaultValue=0.851, parameterName=Cure Remap B}
    Alpha=OneMinus(None=@n14.Result)
   ) {constB=0.851}) {
   atlas=/GrillMoon/Assets/FX/Kaiju/Fight/General/WeaponImpacts/CA_GrillMoon
   curve=&/CC_DemonSlashedMuzzle
   defaultValue=1
   parameterName=ColorCurve
  }
  LinearInterpolate(A=@n1.RGB, B=Multiply(A=@n1.RGB, B=@n15), Alpha=ScalarParameter() {defaultValue=1, parameterName=Use Color Curve})
  n16 = Power(Base=OneMinus(None=@n11), Exp=ScalarParameter() {defaultValue=25, parameterName=Opacity Pow}) {constExponent=25}
  Multiply(
   A=Multiply(A=@n1.A, B=OneMinus(None=@n16))
   B=LinearInterpolate(B=@n15.A, Alpha=ScalarParameter() {parameterName=Use Color Curve Alpha}) {constA=1}
  )
  TextureSample(UVs=@n8) {samplerType=SAMPLERTYPE_LinearGrayscale, texture=&/Substance_graph_height}
 ]
}
