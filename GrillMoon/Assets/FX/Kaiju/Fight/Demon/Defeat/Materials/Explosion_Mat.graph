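# Masked, two-sided lit explosion material for Niagara mesh particles in the demon kaiju defeat effect.
# Curl noise in local space distorts a world-space noise lookup that drives hue-shifted HDR color, dark shading and vertex-normal offset.
# The noise also feeds a temporal-dither mask combined with particle alpha and view-facing falloff.
# Particle dynamic parameters scale emissive and noise, while a PreviousFrameSwitch term builds a previous-frame-only offset for motion blur.
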Material /GrillMoon/Assets/FX/Kaiju/Fight/Demon/Defeat/Materials/Explosion_Mat
graph {
 properties {
  bAutomaticallySetUsageInEditor=false
  bUsedWithNiagaraMeshParticles=true
  blendMode=BLEND_Masked
  twoSided=true
  usedShadingModels=Default Lit
 }
 nodes [
  n1 = Constant()
  n2 = Normalize(VectorInput=Subtract(A=ParticleDirection(), B=ParticlePositionWS()))
  Add(B=Multiply(A=PreviousFrameSwitch(Current Frame=@n1, Previous Frame=@n2), B=ScalarParameter() {defaultValue=1000, parameterName=BlurScale}))
  n3 = DotProduct(A=PixelNormalWS(), B=AtmosphericLightVector())
  Add(A=LinearInterpolate(
    A=ScalarParameter() {defaultValue=0.05, parameterName=Param_9}
    B=ScalarParameter() {defaultValue=0.25, parameterName=Param_8}
    Alpha=Saturate(None=@n3)
   ))
  BlackBody!(Temp)
  ConstantBiasScale(None=MaterialFunctionCall() {materialFunction=&BoundingBoxBased_0-1_UVW}.RGB) {bias=-0.5, scale=2}
  Constant()
  Constant()
  Constant() {r=1}
  n4 = VectorParameter() {parameterName=Param_1, defaultValue=<rgba 1 0 0.0833 1>}
  n5 = TextureSample(UVs=Divide(A=LocalPosition().XYZ, B=ScalarParameter() {defaultValue=2048, parameterName=Param_6})) {samplerType=SAMPLERTYPE_LinearColor, texture=/Game/Effects/FogVolume/VT_CurlNoise_LOW}
  n6 = ScalarParameter() {parameterName=Param_7}
  n7 = WorldPosition()
  n8 = DynamicParameter() {paramNames [Emissive, Noise Scale, Param3, Param4]}
  n9 = TextureSample(UVs=Add(A=Multiply(A=@n5.RGB, B=@n6), B=Divide(A=@n7.XYZ, B=@n8.Noise Scale))) {samplerType=SAMPLERTYPE_LinearColor, texture=/Game/Global/MaterialFunctions/VT_ClothSimWPO}
  n10 = OneMinus(None=@n9.G)
  n11 = MaterialFunctionCall(
   Hue Shift Percentage=ScalarParameter() {parameterName=HueShift}
   Texture=Multiply(A=LinearInterpolate(B=@n4.RGB, Alpha=@n10) {constB=4500}, B=DynamicParameter() {paramNames [Emissive, Noise Scale, Param3, Param4]}.Emissive)
  ) {materialFunction=&HueShift}
  MaterialFunctionCall(
   LDR Input=Saturate(None=@n11.Result)
   HDR Tint=@n11.Result
   HDR Intensity Multiplier=ScalarParameter() {defaultValue=10, parameterName=Param_3}
  ) {materialFunction=&DeriveHDRfromLDR}
  MaterialFunctionCall!(Alpha Threshold, Random) {materialFunction=/Game/Environments/Asteria/Sets/NeonCity/Techart/Screen/MaterialFunctions/DitherTemporalAABlue}
  Multiply(A=LinearInterpolate(A=ScalarParameter() {parameterName=Param_1}, B=ScalarParameter() {defaultValue=1000}, Alpha=@n9.G), B=VertexNormalWS())
  Multiply()
  Multiply()
  Multiply(
   A=VectorParameter() {defaultValue=<rgba 0.02 0.02 0.02 0>}.RGB
   B=LinearInterpolate(A=ScalarParameter() {defaultValue=0.65, parameterName=Param_4}, Alpha=@n9.G)
  )
  n12 = MaterialFunctionCall(Alpha Threshold=ScalarParameter() {defaultValue=0.9, parameterName=Param_5}) {materialFunction=&DitherTemporalAA}
  n13 = DotProduct(A=VertexNormalWS(), B=CameraVectorWS())
  Multiply(A=Multiply(A=ParticleColor().A, B=LinearInterpolate(B=@n12.Result, Alpha=@n9.G) {constA=0.1}), B=Power(Base=Saturate(None=@n13)))
  Multiply(A=@n9.G)
  Saturate(None=Divide(A=PixelDepth().R, B=ScalarParameter() {defaultValue=3500, parameterName=PDO Blending}))
  Saturate(None=Divide(A=PixelDepth().R, B=ScalarParameter() {defaultValue=2048, parameterName=Fade Falloff}))
  ScalarParameter() {parameterName=Param_2}
  VertexNormalWS()
  WorldPosition()
 ]
}
