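# Additive, unlit, two-sided material for Niagara mesh-particle light rays, filed under the GrillMoon Demon kaiju's defeat FX.
# SplineThicken sets ray thickness, with the Tip dynamic parameter driving both base and tip width.
# A V-axis power gradient, offset by the AlongLength dynamic parameter, is multiplied by a U-times-one-minus-U mask that softens the edges.
# A ParticleColor node brings in the particle's color.
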Material /GrillMoon/Assets/FX/Kaiju/Fight/Demon/Defeat/Materials/M_GrillMoon_LightRay_Spline
graph {
 properties {
  bAutomaticallySetUsageInEditor=false
  bUsedWithNiagaraMeshParticles=true
  blendMode=BLEND_Additive
  shadingModel=MSM_Unlit
  twoSided=true
  usedShadingModels=Unlit
 }
 nodes [
  n1 = DynamicParameter() {paramNames [Tip,Base,AlongLength,Param4]}
  MaterialFunctionCall(WidthBase=@n1.Tip, WidthTip=@n1.Tip) {materialFunction=&SplineThicken}
  n2 = TextureCoordinate()
  n3 = ComponentMask(None=@n2) {g=true}
  n4 = ScalarParameter() {defaultValue=4, parameterName=LengthGradient}
  n5 = ComponentMask(None=@n2) {r=true}
  n6 = OneMinus(None=@n5)
  Multiply(
   A=Subtract(A=Power(Base=@n3, Exp=@n4), B=DynamicParameter() {paramNames [Tip,Base,AlongLength,Param4]}.AlongLength)
   B=Power(Base=Multiply(A=@n6, B=@n5), Exp=ScalarParameter() {defaultValue=0.75, parameterName=Edge Gradient})
  )
  ParticleColor()
 ]
}
