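# Translucent two-sided lit smoke trail material for the GrillMoon Kaiju demon defeat FX, enabled for Niagara ribbons and skeletal meshes.
# A panning swirl flowmap distorts a scrolling puffy smoke texture and a sphere mask across V that shapes the trail.
# That mask drives a dark blue tint and opacity, which fades along U with ErodeTrail-controlled erosion.
# Inverted smoke and a second-UV falloff drive a hot pink tint, and smoke also scales a vertex-normal offset.
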Material /GrillMoon/Assets/FX/Kaiju/Fight/Demon/Defeat/Materials/M_GrillMoon_SmokeSpline
graph {
 properties {
  bAutomaticallySetUsageInEditor=false
  bUsedWithMeshDeformer=true
  bUsedWithNiagaraRibbons=true
  bUsedWithSkeletalMesh=true
  blendMode=BLEND_Translucent
  translucencyLightingMode=TLM_VolumetricPerVertexDirectional
  twoSided=true
  usedShadingModels=Default Lit
 }
 nodes [
  n1 = Multiply(A=TextureCoordinate(), B=ComponentMask(None=VectorParameter() {parameterName=C, defaultValue=<rgba 0.5 1 0 0>}.RGB) {g=true, r=true})
  n2 = Multiply(A=TextureCoordinate(), B=ComponentMask(None=VectorParameter() {parameterName=4, defaultValue=<rgba 1 1 0 0>}.RGB) {g=true, r=true})
  n3 = Multiply(A=Time(), B=ScalarParameter() {defaultValue=-0.1, parameterName=Param_3})
  n4 = ComponentMask(None=TextureSample(UVs=Panner(Coordinate=@n2, Time=@n3) {speedX=1}) {samplerType=SAMPLERTYPE_LinearColor, texture=/Game/Global/Textures/Default/Flowmaps/T_SwirlFlow_02}.RGB) {g=true, r=true}
  n5 = Multiply(A=@n4, B=ScalarParameter() {parameterName=Param_5})
  n6 = Time()
  n7 = ScalarParameter() {defaultValue=-0.025, parameterName=Param_9}
  n8 = TextureSample(UVs=Panner(Coordinate=Add(A=@n1, B=@n5), Time=Multiply(A=@n6, B=@n7)) {speedX=1}) {texture=/Game/Effects/Fort_Effects/Textures/Smoke/T_Puffy_Smoke_Trail_01}
  n9 = TextureCoordinate()
  DeriveNormalZ(InXY=Add(A=@n8.G, B=Multiply(A=@n9) {constB=0.5}))
  n10 = Multiply(A=@n4, B=ScalarParameter() {defaultValue=0.05, parameterName=Param_7})
  n11 = ComponentMask(None=TextureCoordinate()) {g=true}
  n12 = LinearInterpolate(A=ScalarParameter() {parameterName=Param_11}, Alpha=@n8.G)
  n13 = Multiply(
   A=SphereMask(
    A=Add(A=@n10, B=@n11)
    B=ScalarParameter() {defaultValue=0.5, parameterName=Param_8}
    Radius=ScalarParameter() {defaultValue=0.5, parameterName=radius}
    Hardness=ScalarParameter() {defaultValue=0.1867, parameterName=hardness}
   )
   B=@n12
  )
  LinearInterpolate(B=VectorParameter() {parameterName=Param_1, defaultValue=<rgba 0.0014 0 0.0799 1>}.RGB, Alpha=@n13)
  n14 = OneMinus(None=@n12)
  n15 = ScalarParameter() {defaultValue=1, parameterName=Param_6}
  n16 = OneMinus(None=ComponentMask(None=TextureCoordinate() {coordinateIndex=1}) {r=true})
  n17 = ScalarParameter() {defaultValue=6, parameterName=Param_2}
  LinearInterpolate(
   B=VectorParameter() {defaultValue=<rgba 5 0 2.9572 1>}.RGB
   Alpha=Multiply(A=Power(Base=@n14, Exp=@n15) {constExponent=4}, B=Power(Base=@n16, Exp=@n17))
  )
  MaterialFunctionCall!(Bump) {materialFunction=/Engine/Functions/Engine_MaterialFunctions03/Procedurals/PerturbNormalLQ}
  Multiply()
  Multiply(A=Multiply(A=@n8.G, B=VertexNormalWS()), B=ScalarParameter())
  n18 = ScalarParameter() {defaultValue=4, parameterName=Opacity}
  n19 = ComponentMask(None=TextureCoordinate()) {r=true}
  n20 = Add(A=@n19, B=Multiply(A=@n8.G, B=ScalarParameter() {defaultValue=0.75, parameterName=Param_12}))
  Multiply(
   A=Multiply(A=Multiply(A=@n13, B=@n18), B=OneMinus(None=@n19))
   B=LinearInterpolate(A=ScalarParameter() {parameterName=ErodeTrail}, Alpha=OneMinus(None=@n20))
  )
  Multiply()
  ScalarParameter() {defaultValue=1, parameterName=Param_4}
 ]
}
