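# Unlit translucent Niagara mesh material for lightning arcs in the Demon kaiju defeat effect, with a bright HDR orange color.
# World position offset stretches strands along UV V toward a target blended between two vector parameters by ArcTravel.
# Curl noise and SplineThicken width both peak mid-strand and taper at the ends.
# Opacity multiplies particle alpha, a Sweep mask along the strand, a Random-driven UV1 mask and panning plasma texture.
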Material /GrillMoon/Assets/FX/Kaiju/Fight/Demon/Defeat/Meshes/M_LightningDeath
graph {
 properties {
  bAutomaticallySetUsageInEditor=false
  bUsedWithNiagaraMeshParticles=true
  blendMode=BLEND_Translucent
  shadingModel=MSM_Unlit
  twoSided=true
  usedShadingModels=Unlit
 }
 nodes [
  n1 = VectorParameter() {parameterName=Param_2, defaultValue=<rgba 0 -500 10 0>}
  n2 = VectorParameter() {parameterName=new location}
  n3 = DynamicParameter() {defaultValue=<rgba 0.5 1 1 1>, paramNames [Random,Sweep,ArcTravel,Param4]}
  n4 = TextureCoordinate()
  n5 = ComponentMask(None=TextureCoordinate()) {g=true}
  n6 = Multiply(A=@n5, B=OneMinus(None=@n5))
  n7 = Multiply(A=WorldPosition().XYZ, B=ScalarParameter() {defaultValue=0.005, parameterName=Noise Scale})
  n8 = Multiply(
   A=ConstantBiasScale(None=TextureSample(UVs=@n7) {samplerType=SAMPLERTYPE_LinearColor, texture=/Game/Effects/FogVolume/VT_CurlNoise_LOW}.RGB) {bias=-0.5, scale=2}
   B=ScalarParameter() {defaultValue=30, parameterName=noise Magnitude}
  )
  n9 = ComponentMask(None=TextureCoordinate()) {g=true}
  n10 = LinearInterpolate(B=ScalarParameter() {defaultValue=850}, Alpha=Multiply(A=OneMinus(None=@n9), B=@n9))
  Add(
   A=LinearInterpolate(B=LinearInterpolate(A=@n1.RGB, B=@n2.RGB, Alpha=@n3.ArcTravel), Alpha=ComponentMask(None=@n4) {g=true})
   B=Add(
    A=Multiply(A=@n6, B=@n8)
    B=MaterialFunctionCall(WidthBase=@n10, WidthTip=@n10) {materialFunction=&SplineThicken}.WorldPositionOffset
   )
  )
  n11 = ComponentMask(None=TextureCoordinate()) {g=true}
  n12 = DynamicParameter() {defaultValue=<rgba 0.5 1 1 1>, paramNames [Random,Sweep,ArcTravel,Param4]}
  n13 = SphereMask(
   A=TextureCoordinate() {coordinateIndex=1}
   B=DynamicParameter() {defaultValue=<rgba 0.5 1 1 1>, paramNames [Random,Sweep,ArcTravel,Param4]}.Random
  ) {attenuationRadius=0.15}
  n14 = Add(A=TextureCoordinate(), B=TextureCoordinate() {coordinateIndex=1})
  n15 = ComponentMask(None=VectorParameter() {parameterName=Param_1, defaultValue=<rgba 1 1 0 0>}.RGB) {g=true, r=true}
  n16 = TextureSample(UVs=Panner(Coordinate=Multiply(A=@n14, B=@n15)) {speedY=-1}) {texture=/Game/Effects/Fort_Effects/Textures/Tile/T_Lightning_Plasma_Soft}
  Multiply(A=ParticleColor().A, B=Multiply(A=SphereMask(A=@n11, B=@n12.Sweep) {attenuationRadius=0.5, hardnessPercent=0}, B=Multiply(A=@n13, B=@n16.G)))
  VectorParameter() {defaultValue=<rgba 3500 1727.0868 0 1>}
 ]
}
