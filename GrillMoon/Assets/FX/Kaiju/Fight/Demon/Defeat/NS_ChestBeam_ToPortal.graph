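# Niagara system for the demon kaiju defeat, sending glowing chest-beam projectiles along curved paths to a fixed portal point.
# Head spawns mesh, sprite and light particles that follow randomized Hermite splines, oriented to their velocity, with user-scaled rate and lifetime.
# Ribbon trails read Head particles and shrink through a region mask at the endpoint, while Streak and BillowySmoke add sprite and ribbon bursts.
# The skeletal-mesh MainTrail and Minimal emitters are disabled.
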NiagaraSystem /GrillMoon/Assets/FX/Kaiju/Fight/Demon/Defeat/NS_ChestBeam_ToPortal
dependencies {
 usedDataInterfaces [
  &NiagaraDataInterfaceCurve
  &NiagaraDataInterfaceColorCurve
  &NiagaraDataInterfaceSpriteRendererInfo
  &NiagaraDataInterfaceVector2DCurve
  &NiagaraDataInterfaceParticleRead
  &NiagaraDataInterfaceSkeletalMesh
 ]
 usedDynamicInputs [&MakeVector2D]
 usedModules [
  &self:SystemSpawnScript.SystemScriptSource.SystemScriptGraph.NiagaraNodeAssignment_0.SetVariables_#1
  &SystemState
 ]
}
emitters {
 BillowySmoke {
  emitterUpdateScript [
   &EmitterState(Life Cycle Mode=::NewEnumerator1, Loop Behavior=::NewEnumerator1, Loop Duration=1, UseLoopDelay=false)
   &SpawnRate(SpawnRate=6, Use Spawn Probability=false)
  ]
  particleSpawnScript [
   &InitializeParticle(
    Lifetime=3
    Position Offset=<500 0 0>
    Ribbon Facing Mode=::NewEnumerator1
    Ribbon Width=2500
    Ribbon Width Mode=::NewEnumerator1
    Sprite Rotation Angle=&RandomRangeFloat(Minimum=-0.25, Maximum=0.25)
    Sprite Rotation Mode=::NewEnumerator1
    Sprite Size Mode=::NewEnumerator1
    Uniform Sprite Size Max=10000
    Uniform Sprite Size Min=9500
    UsePositionOffset=true
    Write Material Random=true
   )
   &AddVelocity(
    Velocity=&RandomRangeVector(Minimum=<17500 -10000 0>, Maximum=<17500 10000 0>)
    Velocity Speed Scale=&RandomRangeFloat(Minimum=1, Maximum=1.1)
   )
  ]
  particleUpdateScript [
   &AccelerationForce(Acceleration=<3500 0 4500>)
   &AccelerationForce(Acceleration=<100 0 0>, Coordinate Space=::Local) {module=AccelerationForce001}
   &Drag(
    Drag=&FloatFromCurve(
     FloatCurve=DataInterface() {propertyValues {Curve {keys [{arriveTangent=2.8621, interpMode=RCIM_Cubic, leaveTangent=2.8621, tangentMode=RCTM_User, time=0, value=0}, <at 1 1>]}}}
     Scale Curve=8
    )
    Use Linear Drag=true
    Use Rotational Drag=true
   )
   &SolveForcesAndVelocity(Clamp Velocity=false, Limit Acceleration=false)
   &ParticleState()
   &SpriteRotationRate()
   &ScaleSpriteSize(Uniform Curve Sprite Scale=DataInterface() {propertyValues {Curve {keys [
        {arriveTangent=4.847, interpMode=RCIM_Cubic, leaveTangent=4.847, tangentMode=RCTM_User, time=0, value=0.15}
        {arriveTangent=0.1006, interpMode=RCIM_Cubic, leaveTangent=0.1006, tangentMode=RCTM_User, time=0.9, value=1}
       ]}}})
   &ScaleColor(
    Linear Color Curve=DataInterface() {propertyValues {
      AlphaCurve {keys [<at 0 1>,<at 1 0>]}
      BlueCurve {keys [<at 0 0.1471>,<at 0.1102 0.1>]}
      GreenCurve {keys [<at 0 0>,<at 0.1102 0>]}
      RedCurve {keys [<at 0 2>,<at 0.1102 0.0398>]}
     }}
    Scale Mode=::NewEnumerator2
   ) {enabled=false}
   &SubUVAnimation(
    "Sprite Renderer "=DataInterface() {propertyValues {SpriteRenderer=&self:BillowySmoke_0.NiagaraSpriteRendererProperties_1}}
    UseEndFrame=false
    UseStartFrame=false
   )
   &DynamicMaterialParameters(
    Index 0 Param 1=&FloatFromCurve(FloatCurve=DataInterface() {propertyValues {Curve {keys [{interpMode=RCIM_Cubic, leaveTangent=3.255, tangentMode=RCTM_Break, time=0, value=1}, <at 1 2>]}}})
    Module.Index0Param0Write=true
    Module.Index0Param1Write=true
    Module.Index0Param2Write=true
    Module.Index0Param3Write=true
   )
   &ScaleColor(
    Linear Color Curve=DataInterface() {propertyValues {
      AlphaCurve {keys [<at 0 1>,<at 1 0>]}
      BlueCurve {keys [<at 0.0014 0.0598>,<at 0.161 0.0025>]}
      GreenCurve {keys [<at 0.0014 0>,<at 0.161 0.0025>]}
      RedCurve {keys [<at 0.0014 0.1>,<at 0.161 0.0025>]}
     }}
    Scale Mode=::NewEnumerator2
   ) {module=ScaleColor001}
  ]
  rendererClasses [&NiagaraRibbonRendererProperties, &NiagaraSpriteRendererProperties]
  renderers [
   {rendererClass=&NiagaraRibbonRendererProperties}
   {rendererClass=&NiagaraSpriteRendererProperties, rendererIndex=1}
  ]
 }
 Head {
  emitterUpdateScript [
   &EmitterState()
   &SpawnRate(SpawnRate=&Multiply_Float(A=1, B=$User.Head Spawn Rate), Use Spawn Probability=true)
  ]
  particleSpawnScript [
   &InitializeParticle(
    Lifetime=&Multiply_Float(A=$User.Head Lifetime, B=&RandomRangeFloat(Minimum=2, Maximum=3.5))
    Mesh Scale=<12 6 6>
    Mesh Scale Mode=::NewEnumerator4
    Sprite Size Mode=::NewEnumerator3
    Uniform Sprite Size=2500
    UsePositionOffset=false
    Write Material Random=true
   )
   &CameraOffset(Camera Offset Amount=50)
  ]
  particleUpdateScript [
   &ParticleState()
   /Game/Effects/Niagara/Modules/Curves/FN_HermiteSpline(
    Compute Velocity=true
    Direction End:Vector3f=&RandomRangeVector(Minimum=<0 -75000 0>, Maximum=<0 -75000 35000>)
    Direction Start:Vector3f=&RandomRangeVector(Minimum=<7500 -31250 -3500>, Maximum=<1500 -90566.8594 3500>)
    Position Blend:NiagaraFloat=$Particles.NormalizedAge
    Position End:Vector3f=<-57600 42800 10500>
    Position Start:Vector3f=/Niagara/DynamicInputs/Helpers/SimulationPosition()
    Write Particles Position=true
   )
   &JitterPosition(Jitter Amount=0) {enabled=false}
   &SolveForcesAndVelocity(Clamp Velocity=false, Limit Acceleration=false)
   /SetVariables(Particles.Position=&Lerp_Vector(
     A=&SimulationPosition()
     B=$Output.FN_HermiteSpline.Position
     Alpha=&ReturnNormalizedExecIndex()
    )) {enabled=false, isSetParameters=true}
   &ScaleColor(
    Linear Color Curve=DataInterface() {propertyValues {
      AlphaCurve {keys [<at 0.5229 1>,<at 1 0>]}
      BlueCurve {keys [<at 0 2.038>,<at 1 0>]}
      GreenCurve {keys [<at 0 0>,<at 1 4.8206>]}
      RedCurve {keys [<at 0 5>,<at 1 10>]}
     }}
    Scale Mode=::NewEnumerator2
   )
   /SetVariables(Particles.MeshOrientation=&FindQuatBetweenTwoVectors(From Vector=<1 0 0>, To Vector=$Output.FN_HermiteSpline.Velocity)) {isSetParameters=true}
   /SetVariables(
    Particles.LightColor=&MultiplyLinearColor(
     Linear Color=$Particles.Color
     Scale Factor=&MakeVector4_FromLinearColor(LinearColor=&ColorFromCurve(ColorCurve=DataInterface() {propertyValues {
         AlphaCurve {keys [<at 0 1 RCIM_Cubic>,<at 1 0 RCIM_Cubic>]}
         BlueCurve {keys [<at 0.6211 0 RCIM_Cubic>,<at 1 0 RCIM_Cubic>]}
         GreenCurve {keys [<at 0.6211 3.75 RCIM_Cubic>,<at 1 0 RCIM_Cubic>]}
         RedCurve {keys [<at 0.6211 5 RCIM_Cubic>,<at 1 0 RCIM_Cubic>]}
        }}))
    )
    Particles.LightRadius=&FloatFromCurve(
     FloatCurve=DataInterface() {propertyValues {Curve {keys [
         {interpMode=RCIM_Cubic, leaveTangent=-0.1087, tangentMode=RCTM_Break, time=0, value=0}
         {interpMode=RCIM_Cubic, tangentMode=RCTM_Break, time=0.7, value=1}
         {interpMode=RCIM_Cubic, tangentMode=RCTM_Break, time=1, value=-0.1}
        ]}}}
     Scale Curve=75
    )
   ) {isSetParameters=true}
  ]
  rendererClasses [
   &NiagaraSpriteRendererProperties
   &NiagaraMeshRendererProperties
   &NiagaraLightRendererProperties
  ]
  renderers [
   {rendererClass=&NiagaraSpriteRendererProperties}
   {rendererClass=&NiagaraMeshRendererProperties, rendererIndex=1}
   {rendererClass=&NiagaraLightRendererProperties, rendererIndex=2}
  ]
 }
 MainTrail {
  bEnabled=false
  emitterUpdateScript [
   &EmitterState()
   &SpawnBurst_Instantaneous(Spawn Count=40, Use Spawn Probability=false, UseLoopCountLimit=true)
   /SetVariables(Emitter.TipFade:NiagaraFloat=$User.Emitter_TipFade) {isSetParameters=true}
  ]
  particleSpawnScript [&InitializeParticle(Ribbon Width=1500, Ribbon Width Mode=::NewEnumerator1, UsePositionOffset=false, Write Material Random=true)]
  particleUpdateScript [
   &ParticleState(
    Kill Particles When Lifetime Has Elapsed=false
    Let Infinitely Lived Particles Die When Emitter Deactivates=true
    Loop Particles Lifetime=true
   )
   &SkeletalMeshLocation(
    Bone / Socket Index=&MakeCustomIntFromBool(Bool=$User.LeftHand, True INT=0, False INT=0)
    Bone Sampling Mode=::NewEnumerator4
    MaskSpawnGroup=false
    Orientation Sampling=::NewEnumerator1
    Skeletal Mesh=$System.Skeletal Mesh
    Write To Particle Position=false
   )
   &FN_HermiteSpline(
    Compute Velocity=true
    Direction End:Vector3f=&MakeCustomVectorFromBool(Bool=true, True Vector=$User.True Vector_Direction End, False Vector=<-280809.3438 0 -4500>)
    Direction Start:Vector3f=&MakeCustomVectorFromBool(
     Bool=$User.LeftHand
     True Vector=&MakeVector(
      X=&Multiply_Float(
       A=40000
       B=&FloatFromCurve(
        FloatCurve=DataInterface() {propertyValues {Curve {keys [{interpMode=RCIM_Cubic, tangentMode=RCTM_User, time=0.2, value=0}, {interpMode=RCIM_Cubic, tangentMode=RCTM_User, time=1, value=1}]}}}
        CurveIndex=&ReturnNormalizedExecIndex()
       )
      )
      Y=0
      Z=27500
     )
     False Vector=<0 0 43750>
    )
    Position Blend:NiagaraFloat=&ReturnNormalizedExecIndex()
    Position End:Vector3f=&SimulationPosition()
    Position Start:Vector3f=<-57600 42800 11500>
    Write Particles Position=true
    Write Particles Velocity=true
   )
   &ScaleRibbonWidth(Ribbon Width Scale=&FloatFromCurve(
     FloatCurve=DataInterface() {propertyValues {Curve {keys [
         {arriveTangent=5.7924, interpMode=RCIM_Cubic, leaveTangent=5.7924, tangentMode=RCTM_User, time=0, value=0}
         {tangentMode=RCTM_User, time=0.3, value=1}
         {arriveTangent=-0.0405, interpMode=RCIM_Cubic, leaveTangent=-0.0405, tangentMode=RCTM_User, time=0.7, value=1}
         {arriveTangent=-4.06, interpMode=RCIM_Cubic, leaveTangent=-4.06, tangentMode=RCTM_User, time=1, value=0.55}
        ]}}}
     CurveIndex=&ReturnNormalizedExecIndex()
     Scale Curve=&FloatFromCurve(
      FloatCurve=DataInterface() {propertyValues {Curve {keys [
          {interpMode=RCIM_Cubic, tangentMode=RCTM_User, time=0, value=0.9}
          {interpMode=RCIM_Cubic, tangentMode=RCTM_User, time=0.6, value=1.175}
          {tangentMode=RCTM_User, time=1, value=0.9}
         ]}}}
      CurveIndex=&Frac_Float(Float=&Add_Float(
        A=&Multiply_Float(A=$Particles.NormalizedAge, B=1.5)
        B=&Multiply_Float(A=&ReturnNormalizedExecIndex(), B=4)
       ))
      Scale Curve=&Lerp_Float(
       A=&FloatFromCurve(
        FloatCurve=DataInterface() {propertyValues {Curve {keys [{tangentMode=RCTM_User, time=0, value=0}, {tangentMode=RCTM_User, time=1, value=1}]}}}
        CurveIndex=&ClampFloat(Float=&Subtract_Float(
          A=&RemapRange(Input Value To Remap=$User.Growth, Input Min=0, Input Max=1, Output Min=0, Output Max=2)
          B=&ReturnNormalizedExecIndex()
         ))
       )
       B=&FloatFromCurve(
        FloatCurve=DataInterface() {propertyValues {Curve {keys [{tangentMode=RCTM_User, time=0, value=1}, {tangentMode=RCTM_User, time=1, value=0}]}}}
        CurveIndex=&ClampFloat(Float=&Subtract_Float(
          A=&RemapRange(Input Value To Remap=$User.Erosion, Input Min=0, Input Max=1, Output Min=0, Output Max=2)
          B=&ReturnNormalizedExecIndex()
         ))
       )
       Alpha=$User.Flip
      )
     )
    ))
   &DynamicMaterialParameters(
    Index 0 Param 1=0.75
    Index 0 Param 3=2.5
    Index 0 Param 4=2
    Index 1 Param 1=50
    Index 2 Param 1=0.35
    Index 2 Param 3=2.5
    Index 2 Param 4=2
    Index 3 Param 1=0.525
    Index 3 Param 3=2.5
    Index 3 Param 4=3
    Module.Index0Param0Write=true
    Module.Index0Param1Write=true
    Module.Index0Param2Write=true
    Module.Index0Param3Write=true
    Module.Index1Param0Write=true
    Module.Index1Param1Write=true
    Module.Index1Param2Write=true
    Module.Index1Param3Write=true
    Module.Index2Param0Write=true
    Module.Index2Param1Write=true
    Module.Index2Param2Write=true
    Module.Index2Param3Write=true
    Module.Index3Param0Write=true
    Module.Index3Param1Write=true
    Module.Index3Param2Write=true
    Module.Index3Param3Write=true
    Write Parameter Index 1=true
    Write Parameter Index 2=true
    Write Parameter Index 3=true
   )
   &ScaleColor(
    Curve Index=&ReturnNormalizedExecIndex()
    Linear Color Curve=DataInterface() {propertyValues {
      AlphaCurve {keys [{interpMode=RCIM_Cubic, leaveTangent=-0.8781, tangentMode=RCTM_Break, time=0, value=1}]}
      BlueCurve {keys [
        {interpMode=RCIM_Cubic, leaveTangent=-0.8781, tangentMode=RCTM_Break, time=0, value=0}
        {arriveTangent=-4.9701, interpMode=RCIM_Cubic, tangentMode=RCTM_Break, time=0.6212, value=2.8551}
       ]}
      GreenCurve {keys [
        {interpMode=RCIM_Cubic, leaveTangent=-0.8781, tangentMode=RCTM_Break, time=0, value=27.9127}
        {arriveTangent=-4.9701, interpMode=RCIM_Cubic, tangentMode=RCTM_Break, time=0.6212, value=0}
       ]}
      RedCurve {keys [
        {interpMode=RCIM_Cubic, leaveTangent=-0.8781, tangentMode=RCTM_Break, time=0, value=50}
        {arriveTangent=-4.9701, interpMode=RCIM_Cubic, tangentMode=RCTM_Break, time=0.6212, value=8}
       ]}
     }}
    Scale Mode=::NewEnumerator2
   )
   /SetVariables(
    Particles.InteriorRibbons:NiagaraFloat=&Add_Float(A=$Particles.RibbonWidth, B=-1000)
    Particles.Translucent Tube Scale:NiagaraFloat=&Add_Float(A=$Particles.RibbonWidth, B=-500)
   ) {isSetParameters=true}
   /SetVariables(Particles.Color=&Multiply_LinearColorByFloat(
     Linear Color=$Particles.Color
     Float=&FloatFromCurve(
      FloatCurve=DataInterface() {propertyValues {Curve {keys [
          {interpMode=RCIM_Cubic, leaveTangent=3.8714, tangentMode=RCTM_Break, time=0, value=1}
          {arriveTangent=0.1591, interpMode=RCIM_Cubic, tangentMode=RCTM_Break, time=0.5, value=2}
          {arriveTangent=-4.2233, interpMode=RCIM_Cubic, tangentMode=RCTM_Break, time=1, value=1}
         ]}}}
      CurveIndex=&Multiply_Float(A=&ReturnNormalizedExecIndex(), B=&ModuloFloat(A=$Particles.Age, B=1))
      Scale Curve=350000
     )
    )) {enabled=false, isSetParameters=true}
  ]
  rendererClasses [&NiagaraRibbonRendererProperties, &NiagaraSpriteRendererProperties]
  renderers [
   {rendererClass=&NiagaraRibbonRendererProperties}
   {rendererClass=&NiagaraSpriteRendererProperties, rendererIndex=1}
   {rendererClass=&NiagaraRibbonRendererProperties, rendererIndex=2}
   {rendererClass=&NiagaraRibbonRendererProperties, rendererIndex=3}
  ]
 }
 Minimal {
  bEnabled=false
  emitterUpdateScript [&EmitterState()]
  particleSpawnScript [&InitializeParticle(Color Mode=::NewEnumerator0, Lifetime=5, UsePositionOffset=false, Write Material Random=true)]
  particleUpdateScript [&ParticleState()]
  rendererClasses [&NiagaraSpriteRendererProperties]
  renderers [{rendererClass=&NiagaraSpriteRendererProperties}]
 }
 Ribbon {
  emitterUpdateScript [
   &EmitterState()
   &SpawnParticlesFromOtherEmitter(
    Attribute Reader=DataInterface() {propertyValues {EmitterBinding {bindingMode=Other, emitterName=Head}}}
    Spawn Rate=30
    Use Maximum Age=false
    Use Minimum Age=false
   )
  ]
  particleSpawnScript [
   &InitializeParticle(
    Color=<rgba 5 5 5 1>
    Lifetime=1.5
    Ribbon Width=250
    Ribbon Width Mode=::NewEnumerator1
    UsePositionOffset=false
    Write Material Random=true
   )
   &SampleParticlesFromOtherEmitter(Particle ID Sampling=::NewEnumerator1)
  ]
  particleUpdateScript [
   &SolveForcesAndVelocity(Clamp Velocity=false, Limit Acceleration=false)
   &ParticleState()
   &ScaleRibbonWidth(Ribbon Width Scale=&FloatFromCurve(FloatCurve=DataInterface() {propertyValues {Curve {keys [{interpMode=RCIM_Cubic, leaveTangent=14.6772, tangentMode=RCTM_Break, time=0, value=0}, <at 0.7 4.5>]}}}))
   &ScaleColor(
    Linear Color Curve=DataInterface() {propertyValues {
      AlphaCurve {keys [<at 0 1>,<at 0.4886 0>]}
      BlueCurve {keys [<at 0 0.6414>,<at 0.4854 0>]}
      GreenCurve {keys [<at 0 0>,<at 0.4854 0.4836>]}
      RedCurve {keys [<at 0 1>,<at 0.4854 1>]}
     }}
    Scale Mode=::NewEnumerator2
   )
   &DynamicMaterialParameters(
    Index 0 Param 1=&RandomRangeFloat(Minimum=0, Maximum=1)
    Module.Index0Param0Write=true
    Module.Index0Param1Write=true
    Module.Index0Param2Write=true
    Module.Index0Param3Write=true
   )
   &RegionMask(Region Falloff=2, Region Origin:Vector3f=<-57600 42800 10500>, Region Radius=5500, UseOffset=false, UseRotation=false)
   &ScaleRibbonWidth(Ribbon Width Scale=&Lerp_Float(A=1, B=0, Alpha=$Output.RegionMask.RegionMask)) {module=ScaleRibbonWidth001}
   &ScaleColor(Curve Index=$Output.RegionMask.RegionMask, Scale Mode=::NewEnumerator2) {module=ScaleColor001}
  ]
  rendererClasses [&NiagaraRibbonRendererProperties]
  renderers [
   {rendererClass=&NiagaraRibbonRendererProperties}
   {rendererClass=&NiagaraRibbonRendererProperties, rendererIndex=1}
  ]
 }
 Streak {
  emitterUpdateScript [
   &EmitterState(Life Cycle Mode=::NewEnumerator1, Loop Behavior=::NewEnumerator1, Loop Duration=1, UseLoopDelay=false)
   &SpawnRate(SpawnRate=32, Use Spawn Probability=false)
  ]
  particleSpawnScript [
   &InitializeParticle(
    Color=<rgba 1 0.5434 0.7589 2>
    Lifetime=0.5
    Ribbon Facing Mode=::NewEnumerator1
    Ribbon Width=2500
    Ribbon Width Mode=::NewEnumerator1
    Sprite Rotation Mode=::NewEnumerator1
    Sprite Size=<3000 12500>
    Sprite Size Mode=::NewEnumerator4
    Sprite UV Mode=::NewEnumerator1
    UsePositionOffset=false
    Write Material Random=true
   )
   &AddVelocity(Velocity=&RandomRangeVector(Minimum=<0.1 -0.1 0>, Maximum=<1 0.1 0>))
  ]
  particleUpdateScript [
   &Drag(
    Drag=&FloatFromCurve(
     FloatCurve=DataInterface() {propertyValues {Curve {keys [{arriveTangent=2.8621, interpMode=RCIM_Cubic, leaveTangent=2.8621, tangentMode=RCTM_User, time=0, value=0}, <at 1 1>]}}}
     Scale Curve=4
    )
    Use Linear Drag=true
    Use Rotational Drag=true
   )
   &SolveForcesAndVelocity(Clamp Velocity=false, Limit Acceleration=false)
   &ParticleState()
   &ScaleSpriteSize(
    Non-Uniform Curve Sprite Scale=DataInterface() {propertyValues {
      XCurve {keys [{interpMode=RCIM_Cubic, leaveTangent=2.0692, tangentMode=RCTM_Break, time=0, value=0.2}, {tangentMode=RCTM_User, time=1, value=1.55}]}
      YCurve {keys [{interpMode=RCIM_Cubic, leaveTangent=1.8377, tangentMode=RCTM_Break, time=0, value=0.2}, {tangentMode=RCTM_User, time=1, value=0.8}]}
     }}
    Scale Sprite Size Mode=::NewEnumerator3
   )
   &ScaleColor(
    Linear Color Curve=DataInterface() {propertyValues {
      AlphaCurve {keys [<at 0 1>,<at 1 0>]}
      BlueCurve {keys [<at 0 4.6689>,<at 0.3761 0.05>]}
      GreenCurve {keys [<at 0 0>,<at 0.3761 0.05>]}
      RedCurve {keys [<at 0 8>,<at 0.3761 0.05>]}
     }}
    Scale Mode=::NewEnumerator2
   )
   &DynamicMaterialParameters(
    Index 0 Param 1=1
    Index 0 Param 2=&FloatFromCurve(FloatCurve=DataInterface() {propertyValues {Curve {keys [{interpMode=RCIM_Cubic, leaveTangent=-1.2147, tangentMode=RCTM_Break, time=0, value=1}, <at 1 0>]}}})
    Module.Index0Param0Write=true
    Module.Index0Param1Write=true
    Module.Index0Param2Write=true
    Module.Index0Param3Write=true
   )
   &SubUVAnimation(
    "Sprite Renderer "=DataInterface() {propertyValues {SpriteRenderer=&self:Streak_3.NiagaraSpriteRendererProperties_1}}
    SubUV Animation Mode:ENiagara_SubUVLookupModeV2=::NewEnumerator5
   )
  ]
  rendererClasses [&NiagaraRibbonRendererProperties, &NiagaraSpriteRendererProperties]
  renderers [
   {rendererClass=&NiagaraRibbonRendererProperties}
   {rendererClass=&NiagaraSpriteRendererProperties, rendererIndex=1}
  ]
 }
}
summary {
 systemName=NS_ChestBeam_ToPortal
 emitters [
  {
   bEnabled=false
   emitterName=MainTrail
   rendererClasses [&NiagaraRibbonRendererProperties, &NiagaraSpriteRendererProperties]
  }
  {bEnabled=false, emitterName=Minimal, rendererClasses [&NiagaraSpriteRendererProperties]}
  {emitterName=BillowySmoke, rendererClasses [&NiagaraRibbonRendererProperties, &NiagaraSpriteRendererProperties]}
  {emitterName=Streak, rendererClasses [&NiagaraRibbonRendererProperties, &NiagaraSpriteRendererProperties]}
  {
   emitterName=Head
   rendererClasses [
    &NiagaraSpriteRendererProperties
    &NiagaraMeshRendererProperties
    &NiagaraLightRendererProperties
   ]
  }
  {emitterName=Ribbon, rendererClasses [&NiagaraRibbonRendererProperties]}
 ]
 userVariables [
  User.Emitter_TipFade {type=NiagaraFloat, defaultValue=2, description=""}
  User.Erosion {type=NiagaraFloat, defaultValue=0, description=""}
  User.Flip {type=NiagaraFloat, defaultValue=0, description=""}
  User.Growth {type=NiagaraFloat, defaultValue=1, description=""}
  User.Head Lifetime {type=NiagaraFloat, defaultValue=1, description=""}
  User.Head Spawn Rate {type=NiagaraFloat, defaultValue=1, description=""}
  User.LeftHand {type=NiagaraBool, defaultValue=true}
  User.RibbonAlpha {type=NiagaraFloat, defaultValue=1, description=""}
  User.True Vector_Direction End {type=Vector3f, description="", defaultValue {struct=Vector3f, value=<-329812.6875 -90049.1797 -104875.8203>}}
 ]
}
