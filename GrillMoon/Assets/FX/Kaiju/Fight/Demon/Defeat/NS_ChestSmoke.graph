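# Niagara smoke effect, apparently for the chest in the Demon kaiju's defeat, with billowing sprite smoke and short magenta streaks that darken as they fade.
# MainTrail lays a burst of particles along a Hermite spline to the simulation position, with LeftHand switching the start tangent.
# User Growth, Erosion and Flip sweep ribbon width along the trail, while a color pulse scrolls through it.
# Sphere_Main and Minimal emitters are disabled.
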NiagaraSystem /GrillMoon/Assets/FX/Kaiju/Fight/Demon/Defeat/NS_ChestSmoke
dependencies {
 usedDataInterfaces [
  &NiagaraDataInterfaceCurve
  &NiagaraDataInterfaceColorCurve
  &NiagaraDataInterfaceSpriteRendererInfo
  &NiagaraDataInterfaceVector2DCurve
 ]
 usedDynamicInputs [&MakeVector2D]
 usedModules [&SystemState]
}
emitters {
 BillowySmoke {
  emitterUpdateScript [
   &EmitterState()
   &SpawnRate(SpawnRate=6, Use Spawn Probability=false)
  ]
  particleSpawnScript [
   &InitializeParticle(
    Lifetime=3
    Position Offset=<500 0 0>
    Ribbon Facing Mode=::NewEnumerator1
    Ribbon Width=2500
    Ribbon Width Mode=::NewEnumerator1
    Sprite Rotation Angle=&RandomRangeFloat(Minimum=-0.25, Maximum=0.25)
    Sprite Rotation Mode=::NewEnumerator1
    Sprite Size Mode=::NewEnumerator1
    Uniform Sprite Size Max=14500
    Uniform Sprite Size Min=11500
    UsePositionOffset=true
    Write Material Random=true
   )
   &AddVelocity(
    Velocity=&RandomRangeVector(Minimum=<17500 -10000 0>, Maximum=<17500 10000 0>)
    Velocity Speed Scale=&RandomRangeFloat(Minimum=1, Maximum=1.1)
   )
  ]
  particleUpdateScript [
   &AccelerationForce(Acceleration=<0 0 4500>)
   &AccelerationForce(Acceleration=<100 0 0>, Coordinate Space=::Local) {module=AccelerationForce001}
   &Drag(
    Drag=&FloatFromCurve(
     FloatCurve=DataInterface() {propertyValues {Curve {keys [{arriveTangent=2.8621, interpMode=RCIM_Cubic, leaveTangent=2.8621, tangentMode=RCTM_User, time=0, value=0}, <at 1 1>]}}}
     Scale Curve=8
    )
    Use Linear Drag=true
    Use Rotational Drag=true
   )
   &SolveForcesAndVelocity(Clamp Velocity=false, Limit Acceleration=false)
   &ParticleState()
   &SpriteRotationRate()
   &ScaleSpriteSize(Uniform Curve Sprite Scale=DataInterface() {propertyValues {Curve {keys [
        {arriveTangent=4.847, interpMode=RCIM_Cubic, leaveTangent=4.847, tangentMode=RCTM_User, time=0, value=0.15}
        {arriveTangent=0.1006, interpMode=RCIM_Cubic, leaveTangent=0.1006, tangentMode=RCTM_User, time=0.9, value=1}
       ]}}})
   &ScaleColor(
    Linear Color Curve=DataInterface() {propertyValues {
      AlphaCurve {keys [<at 0 1>,<at 1 0>]}
      BlueCurve {keys [<at 0 0.1471>,<at 0.1102 0.1>]}
      GreenCurve {keys [<at 0 0>,<at 0.1102 0>]}
      RedCurve {keys [<at 0 2>,<at 0.1102 0.0398>]}
     }}
    Scale Mode=::NewEnumerator2
   ) {enabled=false}
   &SubUVAnimation(
    "Sprite Renderer "=DataInterface() {propertyValues {SpriteRenderer=&self:BillowySmoke_0.NiagaraSpriteRendererProperties_1}}
    UseEndFrame=false
    UseStartFrame=false
   )
   &DynamicMaterialParameters(
    Index 0 Param 1=&FloatFromCurve(FloatCurve=DataInterface() {propertyValues {Curve {keys [{interpMode=RCIM_Cubic, leaveTangent=3.255, tangentMode=RCTM_Break, time=0, value=1}, <at 1 2>]}}})
    Module.Index0Param0Write=true
    Module.Index0Param1Write=true
    Module.Index0Param2Write=true
    Module.Index0Param3Write=true
   )
   &ScaleColor(
    Linear Color Curve=DataInterface() {propertyValues {
      AlphaCurve {keys [<at 0 1>,<at 1 0>]}
      BlueCurve {keys [<at 0.0014 0.0598>,<at 0.161 0.0025>]}
      GreenCurve {keys [<at 0.0014 0>,<at 0.161 0.0025>]}
      RedCurve {keys [<at 0.0014 0.1>,<at 0.161 0.0025>]}
     }}
    Scale Mode=::NewEnumerator2
   ) {module=ScaleColor001}
  ]
  rendererClasses [&NiagaraRibbonRendererProperties, &NiagaraSpriteRendererProperties]
  renderers [
   {rendererClass=&NiagaraRibbonRendererProperties}
   {rendererClass=&NiagaraSpriteRendererProperties, rendererIndex=1}
  ]
 }
 MainTrail {
  emitterUpdateScript [
   &EmitterState(Life Cycle Mode=::NewEnumerator1, Loop Behavior=::NewEnumerator1, Loop Duration Mode=::NewEnumerator1, UseLoopDelay=false)
   &SpawnBurst_Instantaneous(Spawn Count=40, Use Spawn Probability=false, UseLoopCountLimit=true)
   /SetVariables(Emitter.ErodeSmoke:NiagaraFloat=$User.Emitter_ErodeSmoke, Emitter.SmokeOpacity:NiagaraFloat=$User.Emitter_SmokeOpacity) {isSetParameters=true}
  ]
  particleSpawnScript [&InitializeParticle(
    Color=<rgba 0.1862 0.0124 0.7448 1>
    Lifetime=5
    Ribbon Width=7500
    Ribbon Width Mode=::NewEnumerator1
    UsePositionOffset=false
    Write Material Random=true
   )]
  particleUpdateScript [
   &ParticleState(Kill Particles When Lifetime Has Elapsed=false)
   &SkeletalMeshLocation(
    Bone / Socket Index=&MakeCustomIntFromBool(Bool=$User.LeftHand, True INT=1, False INT=0)
    Bone Sampling Mode=::NewEnumerator4
    MaskSpawnGroup=false
    Orientation Sampling=::NewEnumerator1
    Skeletal Mesh=$System.Skeletal Mesh
    Write To Particle Position=false
   )
   /Game/Effects/Niagara/Modules/Curves/FN_HermiteSpline(
    Direction End:Vector3f=<0 0 -3649.8193>
    Direction Start:Vector3f=&MakeCustomVectorFromBool(Bool=$User.LeftHand, True Vector=<-37706.7617 -121619.9375 -82738.5938>, False Vector=<0 -3270649.25 16.2781>)
    Position Blend:NiagaraFloat=&ReturnNormalizedExecIndex()
    Position End:Vector3f=/Niagara/DynamicInputs/Helpers/SimulationPosition()
    Position Start:Vector3f=<-7688.2183 64008.1484 54883.8516>
    Write Particles Position=true
   )
   &ScaleRibbonWidth(Ribbon Width Scale=&FloatFromCurve(
     FloatCurve=DataInterface() {propertyValues {Curve {keys [
         {arriveTangent=-0.1824, interpMode=RCIM_Cubic, leaveTangent=-0.1824, tangentMode=RCTM_User, time=0, value=4}
         {arriveTangent=-4.0528, interpMode=RCIM_Cubic, leaveTangent=-3.6567, tangentMode=RCTM_Break, time=0.6, value=1.75}
         {arriveTangent=-0.1088, interpMode=RCIM_Cubic, leaveTangent=-0.1088, tangentMode=RCTM_User, time=1.3, value=0}
        ]}}}
     CurveIndex=&ReturnNormalizedExecIndex()
     Scale Curve=&FloatFromCurve(
      FloatCurve=DataInterface() {propertyValues {Curve {keys [
          {interpMode=RCIM_Cubic, tangentMode=RCTM_User, time=0, value=1}
          {interpMode=RCIM_Cubic, tangentMode=RCTM_User, time=0.5, value=1}
          {tangentMode=RCTM_User, time=1, value=1}
         ]}}}
      CurveIndex=&Frac_Float(Float=&Add_Float(
        A=&Multiply_Float(A=$Particles.NormalizedAge, B=-2)
        B=&Multiply_Float(A=&ReturnNormalizedExecIndex(), B=4)
       ))
      Scale Curve=&Lerp_Float(
       A=&FloatFromCurve(
        FloatCurve=DataInterface() {propertyValues {Curve {keys [{tangentMode=RCTM_User, time=0, value=0}, {tangentMode=RCTM_User, time=1, value=1}]}}}
        CurveIndex=&ClampFloat(Float=&Subtract_Float(
          A=&RemapRange(Input Value To Remap=$User.Growth, Input Min=0, Input Max=1, Output Min=0, Output Max=2)
          B=&ReturnNormalizedExecIndex()
         ))
       )
       B=&FloatFromCurve(
        FloatCurve=DataInterface() {propertyValues {Curve {keys [{tangentMode=RCTM_User, time=0, value=1}, {tangentMode=RCTM_User, time=1, value=0}]}}}
        CurveIndex=&ClampFloat(Float=&Subtract_Float(
          A=&RemapRange(Input Value To Remap=$User.Erosion, Input Min=0, Input Max=1, Output Min=0, Output Max=2)
          B=&ReturnNormalizedExecIndex()
         ))
       )
       Alpha=$User.Flip
      )
     )
    ))
   &DynamicMaterialParameters(
    Index 0 Param 1=0.6
    Index 0 Param 3=3.25
    Index 0 Param 4=2
    Index 1 Param 1=50
    Index 2 Param 1=0.35
    Index 2 Param 3=3.25
    Index 2 Param 4=2
    Index 3 Param 1=0.525
    Index 3 Param 3=3.25
    Index 3 Param 4=3
    Module.Index0Param0Write=true
    Module.Index0Param1Write=true
    Module.Index0Param2Write=true
    Module.Index0Param3Write=true
    Module.Index1Param0Write=true
    Module.Index1Param1Write=true
    Module.Index1Param2Write=true
    Module.Index1Param3Write=true
    Module.Index2Param0Write=true
    Module.Index2Param1Write=true
    Module.Index2Param2Write=true
    Module.Index2Param3Write=true
    Module.Index3Param0Write=true
    Module.Index3Param1Write=true
    Module.Index3Param2Write=true
    Module.Index3Param3Write=true
    Write Parameter Index 1=true
    Write Parameter Index 2=true
    Write Parameter Index 3=true
   )
   &ScaleColor(
    Scale RGB=&MakeVectorFromLinearColorRGB(Color=&ColorFromCurve(
      ColorCurve=DataInterface() {propertyValues {
        AlphaCurve {keys [<at 0 1 RCIM_Cubic>,<at 1 1>]}
        BlueCurve {keys [<at 0 0.0194 RCIM_Cubic>,<at 0.5 0.1268 RCIM_Cubic>,<at 1 0.0194 RCIM_Cubic>]}
        GreenCurve {keys [<at 0 0.0131 RCIM_Cubic>,<at 0.5 0 RCIM_Cubic>,<at 1 0.0131 RCIM_Cubic>]}
        RedCurve {keys [<at 0 0.07 RCIM_Cubic>,<at 0.5 1.2 RCIM_Cubic>,<at 1 0.07 RCIM_Cubic>]}
       }}
      CurveIndex=&Frac_Float(Float=&Add_Float(
        A=&Multiply_Float(A=$Particles.NormalizedAge, B=2)
        B=&Multiply_Float(A=&ReturnNormalizedExecIndex(), B=3)
       ))
     ))
    ScaleA=true
    ScaleRGB=true
   )
   /SetVariables(
    Particles.InnerColor:LinearColor=<rgba 0.0126 0 0.02 1>
    Particles.InnerRibbonSize:NiagaraFloat=&FloatFromCurve(
     FloatCurve=DataInterface() {propertyValues {Curve {keys [
         {arriveTangent=6.1118, interpMode=RCIM_Cubic, leaveTangent=6.1118, tangentMode=RCTM_User, time=0, value=0.4}
         {tangentMode=RCTM_User, time=0.3, value=1}
         {arriveTangent=-0.0405, interpMode=RCIM_Cubic, leaveTangent=-0.0405, tangentMode=RCTM_User, time=0.7, value=1}
         {arriveTangent=-4.06, interpMode=RCIM_Cubic, leaveTangent=-4.06, tangentMode=RCTM_User, time=1, value=0.5}
        ]}}}
     CurveIndex=&ReturnNormalizedExecIndex()
     Scale Curve=&FloatFromCurve(
      FloatCurve=DataInterface() {propertyValues {Curve {keys [
          {interpMode=RCIM_Cubic, tangentMode=RCTM_User, time=0, value=0.9}
          {interpMode=RCIM_Cubic, tangentMode=RCTM_User, time=0.5, value=1.5}
          {tangentMode=RCTM_User, time=1, value=0.9}
         ]}}}
      CurveIndex=&Frac_Float(Float=&Add_Float(
        A=&Multiply_Float(A=$Particles.NormalizedAge, B=-2)
        B=&Multiply_Float(A=&ReturnNormalizedExecIndex(), B=4)
       ))
      Scale Curve=&Lerp_Float(
       A=&FloatFromCurve(
        FloatCurve=DataInterface() {propertyValues {Curve {keys [{tangentMode=RCTM_User, time=0, value=0}, {tangentMode=RCTM_User, time=1, value=1}]}}}
        CurveIndex=&ClampFloat(Float=&Subtract_Float(
          A=&RemapRange(Input Value To Remap=$User.Growth, Input Min=0, Input Max=1, Output Min=0, Output Max=2)
          B=&ReturnNormalizedExecIndex()
         ))
        Scale Curve=1750
       )
       B=&FloatFromCurve(
        FloatCurve=DataInterface() {propertyValues {Curve {keys [{tangentMode=RCTM_User, time=0, value=1}, {tangentMode=RCTM_User, time=1, value=0}]}}}
        CurveIndex=&ClampFloat(Float=&Subtract_Float(
          A=&RemapRange(Input Value To Remap=$User.Erosion, Input Min=0, Input Max=1, Output Min=0, Output Max=2)
          B=&ReturnNormalizedExecIndex()
         ))
        Scale Curve=1750
       )
       Alpha=$User.Flip
      )
     )
    )
    Particles.SecondaryColor:LinearColor=<rgba 0.0781 0 0.0771 1>
    Particles.SecondaryRibbonSize:NiagaraFloat=&FloatFromCurve(
     FloatCurve=DataInterface() {propertyValues {Curve {keys [
         {arriveTangent=5.6931, interpMode=RCIM_Cubic, leaveTangent=5.6931, tangentMode=RCTM_User, time=0, value=0.4}
         {tangentMode=RCTM_User, time=0.3, value=1}
         {arriveTangent=-0.0405, interpMode=RCIM_Cubic, leaveTangent=-0.0405, tangentMode=RCTM_User, time=0.7, value=1}
         {arriveTangent=-4.06, interpMode=RCIM_Cubic, leaveTangent=-4.06, tangentMode=RCTM_User, time=1, value=0.5}
        ]}}}
     CurveIndex=&ReturnNormalizedExecIndex()
     Scale Curve=&FloatFromCurve(
      FloatCurve=DataInterface() {propertyValues {Curve {keys [
          {interpMode=RCIM_Cubic, tangentMode=RCTM_User, time=0, value=0.9}
          {interpMode=RCIM_Cubic, tangentMode=RCTM_User, time=0.5, value=1.5}
          {tangentMode=RCTM_User, time=1, value=0.9}
         ]}}}
      CurveIndex=&Frac_Float(Float=&Add_Float(
        A=&Multiply_Float(A=$Particles.NormalizedAge, B=-2)
        B=&Multiply_Float(A=&ReturnNormalizedExecIndex(), B=4)
       ))
      Scale Curve=&Lerp_Float(
       A=&FloatFromCurve(
        FloatCurve=DataInterface() {propertyValues {Curve {keys [{tangentMode=RCTM_User, time=0, value=0}, {tangentMode=RCTM_User, time=1, value=1}]}}}
        CurveIndex=&ClampFloat(Float=&Subtract_Float(
          A=&RemapRange(Input Value To Remap=$User.Growth, Input Min=0, Input Max=1, Output Min=0, Output Max=2)
          B=&ReturnNormalizedExecIndex()
         ))
        Scale Curve=2100
       )
       B=&FloatFromCurve(
        FloatCurve=DataInterface() {propertyValues {Curve {keys [{tangentMode=RCTM_User, time=0, value=1}, {tangentMode=RCTM_User, time=1, value=0}]}}}
        CurveIndex=&ClampFloat(Float=&Subtract_Float(
          A=&RemapRange(Input Value To Remap=$User.Erosion, Input Min=0, Input Max=1, Output Min=0, Output Max=2)
          B=&ReturnNormalizedExecIndex()
         ))
        Scale Curve=2000
       )
       Alpha=$User.Flip
      )
     )
    )
    Particles.SoftAirColor:LinearColor=&MultiplyLinearColor(
     Linear Color=<rgba 0.1302 0 0.0386 0.2>
     Scale Factor=&MakeFloat4FromVectorAndFloat(
      "Vector (RGB)"=<1 1 1>
      "Float (Alpha)"=&FloatFromCurve(
       FloatCurve=DataInterface() {propertyValues {Curve {keys [{interpMode=RCIM_Cubic, leaveTangent=3.4808, tangentMode=RCTM_Break, time=0, value=0}, {tangentMode=RCTM_User, time=1, value=1}]}}}
       CurveIndex=&FloatFromCurve(
        FloatCurve=DataInterface() {propertyValues {Curve {keys [{tangentMode=RCTM_User, time=0, value=0}, {tangentMode=RCTM_User, time=1, value=1}]}}}
        CurveIndex=&ClampFloat(Float=&Subtract_Float(
          A=&RemapRange(Input Value To Remap=$User.Growth, Input Min=0, Input Max=1, Output Min=0, Output Max=2)
          B=&ReturnNormalizedExecIndex()
         ))
       )
      )
     )
    )
    Particles.SoftAirDynm:Vector4f=<0.5 0 0 0>
    Particles.SoftAirSpriteSize:Vector2f=&Vector2DFromFloat(Value=&FloatFromCurve(
      FloatCurve=DataInterface() {propertyValues {Curve {keys [
          {arriveTangent=5.2686, interpMode=RCIM_Cubic, leaveTangent=5.2686, tangentMode=RCTM_User, time=0, value=0.3}
          {tangentMode=RCTM_User, time=0.3, value=1}
          {arriveTangent=-0.0405, interpMode=RCIM_Cubic, leaveTangent=-0.0405, tangentMode=RCTM_User, time=0.7, value=1}
          {arriveTangent=-4.06, interpMode=RCIM_Cubic, leaveTangent=-4.06, tangentMode=RCTM_User, time=1, value=0.5}
         ]}}}
      CurveIndex=&ReturnNormalizedExecIndex()
      Scale Curve=&Lerp_Float(
       A=&FloatFromCurve(
        FloatCurve=DataInterface() {propertyValues {Curve {keys [{tangentMode=RCTM_User, time=0, value=0}, {tangentMode=RCTM_User, time=1, value=1}]}}}
        CurveIndex=&ClampFloat(Float=&Subtract_Float(
          A=&RemapRange(Input Value To Remap=$User.Growth, Input Min=0, Input Max=1, Output Min=0, Output Max=2)
          B=&ReturnNormalizedExecIndex()
         ))
        Scale Curve=3500
       )
       B=&FloatFromCurve(
        FloatCurve=DataInterface() {propertyValues {Curve {keys [{tangentMode=RCTM_User, time=0, value=1}, {tangentMode=RCTM_User, time=1, value=0}]}}}
        CurveIndex=&ClampFloat(Float=&Subtract_Float(
          A=&RemapRange(Input Value To Remap=$User.Erosion, Input Min=0, Input Max=1, Output Min=0, Output Max=2)
          B=&ReturnNormalizedExecIndex()
         ))
        Scale Curve=3500
       )
       Alpha=$User.Flip
      )
     ))
   ) {enabled=false, isSetParameters=true}
  ]
  rendererClasses [&NiagaraRibbonRendererProperties, &NiagaraSpriteRendererProperties]
  renderers [
   {rendererClass=&NiagaraRibbonRendererProperties}
   {rendererClass=&NiagaraRibbonRendererProperties, rendererIndex=1}
   {rendererClass=&NiagaraRibbonRendererProperties, rendererIndex=2}
   {rendererClass=&NiagaraSpriteRendererProperties, rendererIndex=3}
  ]
 }
 Minimal {
  bEnabled=false
  emitterUpdateScript [&EmitterState()]
  particleSpawnScript [&InitializeParticle(Color Mode=::NewEnumerator0, Lifetime=5, UsePositionOffset=false, Write Material Random=true)]
  particleUpdateScript [
   &ParticleState()
   &FN_HermiteSpline()
  ]
  rendererClasses [&NiagaraSpriteRendererProperties]
  renderers [{rendererClass=&NiagaraSpriteRendererProperties}]
 }
 Sphere_Main {
  bEnabled=false
  emitterUpdateScript [
   &EmitterState()
   &SpawnRate(SpawnRate=12, Use Spawn Probability=false)
  ]
  particleSpawnScript [
   &InitializeParticle(Mesh Scale Mode=::NewEnumerator1, Mesh Uniform Scale Max=29, Mesh Uniform Scale Min=12, UsePositionOffset=false, Write Material Random=true)
   &AddVelocity(
    Velocity=&RandomRangeVector(Minimum=<21500 -10000 0>, Maximum=<22500 15000 1>)
    Velocity Speed Scale=&RandomRangeFloat(Minimum=1, Maximum=1.2)
   )
  ]
  particleUpdateScript [
   &AccelerationForce(Acceleration=<17500 0 3500>)
   &AccelerationForce(Acceleration=<0 0 500>) {enabled=false, module=AccelerationForce001}
   &Drag(
    Drag=&FloatFromCurve(FloatCurve=DataInterface() {propertyValues {Curve {keys [{interpMode=RCIM_Cubic, leaveTangent=30.0988, tangentMode=RCTM_Break, time=0, value=1}, <at 0.25 4>]}}})
    Use Linear Drag=true
    Use Rotational Drag=true
   )
   &SolveForcesAndVelocity(Clamp Velocity=false, Limit Acceleration=false)
   &ParticleState()
   &ScaleMeshSizeBySpeed(Max Scale Factor=<1.5 1 1>, Min Scale Factor:Vector3f=<0.05 0.05 0.05>)
   &ScaleMeshSize(Scale Factor=&VectorFromFloat(Value=&FloatFromCurve(FloatCurve=DataInterface() {propertyValues {Curve {keys [
          {interpMode=RCIM_Cubic, leaveTangent=4.0682, tangentMode=RCTM_Break, time=0, value=0.65}
          {arriveTangent=0.036, interpMode=RCIM_Cubic, leaveTangent=0.036, tangentMode=RCTM_User, time=0.75, value=2.35}
         ]}}})))
   &DynamicMaterialParameters(
    Index 0 Param 1=5
    Index 0 Param 2=&RandomRangeFloat(Minimum=3072, Maximum=6144)
    Module.Index0Param0Write=true
    Module.Index0Param1Write=true
    Module.Index0Param2Write=true
    Module.Index0Param3Write=true
   )
   &ScaleColor(
    Linear Color Curve=DataInterface() {propertyValues {
      AlphaCurve {keys [<at 0 2>,<at 1 0>]}
      BlueCurve {keys [<at 0 1>,<at 1 1>]}
      GreenCurve {keys [<at 0 1>,<at 1 1>]}
      RedCurve {keys [<at 0 1>,<at 1 1>]}
     }}
    Scale Mode=::NewEnumerator2
   )
  ]
  rendererClasses [&NiagaraMeshRendererProperties]
  renderers [{rendererClass=&NiagaraMeshRendererProperties, rendererIndex=0}]
 }
 Streak {
  emitterUpdateScript [
   &EmitterState()
   &SpawnRate(SpawnRate=32, Use Spawn Probability=false)
  ]
  particleSpawnScript [
   &InitializeParticle(
    Color=<rgba 1 0.5434 0.7589 2>
    Lifetime=0.5
    Ribbon Facing Mode=::NewEnumerator1
    Ribbon Width=2500
    Ribbon Width Mode=::NewEnumerator1
    Sprite Rotation Mode=::NewEnumerator1
    Sprite Size=<3000 12500>
    Sprite Size Mode=::NewEnumerator4
    Sprite UV Mode=::NewEnumerator1
    UsePositionOffset=false
    Write Material Random=true
   )
   &AddVelocity(Velocity=&RandomRangeVector(Minimum=<0.1 -0.1 0>, Maximum=<1 0.1 0>))
  ]
  particleUpdateScript [
   &Drag(
    Drag=&FloatFromCurve(
     FloatCurve=DataInterface() {propertyValues {Curve {keys [{arriveTangent=2.8621, interpMode=RCIM_Cubic, leaveTangent=2.8621, tangentMode=RCTM_User, time=0, value=0}, <at 1 1>]}}}
     Scale Curve=4
    )
    Use Linear Drag=true
    Use Rotational Drag=true
   )
   &SolveForcesAndVelocity(Clamp Velocity=false, Limit Acceleration=false)
   &ParticleState()
   &ScaleSpriteSize(
    Non-Uniform Curve Sprite Scale=DataInterface() {propertyValues {
      XCurve {keys [{interpMode=RCIM_Cubic, leaveTangent=2.0692, tangentMode=RCTM_Break, time=0, value=0.2}, {tangentMode=RCTM_User, time=1, value=1.55}]}
      YCurve {keys [{interpMode=RCIM_Cubic, leaveTangent=1.8377, tangentMode=RCTM_Break, time=0, value=0.2}, {tangentMode=RCTM_User, time=1, value=0.8}]}
     }}
    Scale Sprite Size Mode=::NewEnumerator3
   )
   &ScaleColor(
    Linear Color Curve=DataInterface() {propertyValues {
      AlphaCurve {keys [<at 0 1>,<at 1 0>]}
      BlueCurve {keys [<at 0 4.6689>,<at 0.3761 0.05>]}
      GreenCurve {keys [<at 0 0>,<at 0.3761 0.05>]}
      RedCurve {keys [<at 0 8>,<at 0.3761 0.05>]}
     }}
    Scale Mode=::NewEnumerator2
   )
   &DynamicMaterialParameters(
    Index 0 Param 1=1
    Index 0 Param 2=&FloatFromCurve(FloatCurve=DataInterface() {propertyValues {Curve {keys [{interpMode=RCIM_Cubic, leaveTangent=-1.2147, tangentMode=RCTM_Break, time=0, value=1}, <at 1 0>]}}})
    Module.Index0Param0Write=true
    Module.Index0Param1Write=true
    Module.Index0Param2Write=true
    Module.Index0Param3Write=true
   )
   &SubUVAnimation(
    "Sprite Renderer "=DataInterface() {propertyValues {SpriteRenderer=&self:Streak_3.NiagaraSpriteRendererProperties_1}}
    SubUV Animation Mode:ENiagara_SubUVLookupModeV2=::NewEnumerator5
   )
  ]
  rendererClasses [&NiagaraRibbonRendererProperties, &NiagaraSpriteRendererProperties]
  renderers [
   {rendererClass=&NiagaraRibbonRendererProperties}
   {rendererClass=&NiagaraSpriteRendererProperties, rendererIndex=1}
  ]
 }
}
summary {
 systemName=NS_ChestSmoke
 emitters [
  {emitterName=BillowySmoke, rendererClasses [&NiagaraRibbonRendererProperties, &NiagaraSpriteRendererProperties]}
  {emitterName=Streak, rendererClasses [&NiagaraRibbonRendererProperties, &NiagaraSpriteRendererProperties]}
  {bEnabled=false, emitterName=Sphere_Main, rendererClasses [&NiagaraMeshRendererProperties]}
  {emitterName=MainTrail, rendererClasses [&NiagaraRibbonRendererProperties, &NiagaraSpriteRendererProperties]}
  {bEnabled=false, emitterName=Minimal, rendererClasses [&NiagaraSpriteRendererProperties]}
 ]
 userVariables [
  User.Emitter_ErodeSmoke {type=NiagaraFloat, defaultValue=0, description=""}
  User.Emitter_SmokeOpacity {type=NiagaraFloat, defaultValue=0, description=""}
  User.Erosion {type=NiagaraFloat, defaultValue=0, description=""}
  User.Flip {type=NiagaraFloat, defaultValue=0, description=""}
  User.Growth {type=NiagaraFloat, defaultValue=1, description=""}
  User.LeftHand {type=NiagaraBool, defaultValue=true}
 ]
}
