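# Niagara effect, apparently for the kaiju demon's defeat, that turns a skeletal mesh into streaming ash.
# GPU Ash sprites spawn on the mesh and ride Hermite splines to a fixed endpoint, cooling from hot magenta.
# Ash001 keeps growing glow sprites and meshes on the surface, culled by normal-to-camera dot product and offset by camera vector.
# A brief CPU Head burst follows the same path, feeding noisy Trails ribbons through particle reads.
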NiagaraSystem /GrillMoon/Assets/FX/Kaiju/Fight/Demon/Defeat/NS_DemonAsh_Mask
dependencies {
 usedDataInterfaces [
  &NiagaraDataInterfaceColorCurve
  &NiagaraDataInterfaceCurve
  &NiagaraDataInterfaceCamera
  &NiagaraDataInterfaceParticleRead
 ]
 usedDynamicInputs []
 usedModules [
  &self:SystemSpawnScript.SystemScriptSource.SystemScriptGraph.NiagaraNodeAssignment_0.SetVariables_#1
  &SystemState
  &self:SystemSpawnScript.SystemScriptSource.SystemScriptGraph.NiagaraNodeAssignment_1.SetVariables_#2
 ]
}
emitters {
 Ash {
  simTarget=GPUComputeSim
  emitterUpdateScript [
   &EmitterState()
   &SpawnRate(SpawnRate=2750, Use Spawn Probability=false)
  ]
  particleSpawnScript [
   &InitializeParticle(
    Color=<rgba 0.1042 0.1042 0.1042 1>
    Lifetime=2.5
    Sprite Size Mode=::NewEnumerator1
    Uniform Sprite Size Max=1000
    Uniform Sprite Size Min=500
    UsePositionOffset=false
    Write Material Random=true
   )
   &SkeletalMeshLocation(
    "Inherit Velocity (Scale)"=0.1
    MaskSpawnGroup=false
    Mesh Sampling Type=::NewEnumerator7
    OffsetAlongNormal=false
    OverrideRandomSeed=false
    Skeletal Mesh=$System.Skeletal Mesh
    Velocity Sampling=::NewEnumerator0
    VelocitySpeedLimit=false
   )
   &KillParticlesInVolume(Sphere Radius=9500)
   &CurlNoiseForce(Bias Noise=false, Noise Frequency=350, Noise Strength=350, Pan Noise=false) {module=CurlNoiseForce001}
   &ApplyInitialForces()
   /SetVariables(Particles.SpriteFacing=&RotateVectorByQuaternion(Vector To Rotate=<1 0 0>, Quaternion=$Particles.SkeletalMeshLocation.SampledRotation)) {isSetParameters=true, module=SetVariables_#3}
  ]
  particleUpdateScript [
   &ParticleState()
   &SpriteRotationRate(Rotation Rate=10)
   /Game/Effects/Niagara/Modules/Curves/FN_HermiteSpline(
    Compute Acceleration=true
    Direction End:Vector3f=<0 -75000 15000>
    Position Blend:NiagaraFloat=$Particles.NormalizedAge
    Position End:Vector3f=<-57600 42904 10500>
    Position Start:Vector3f=$Particles.SkeletalMeshLocation.SampledPosition
    Write Particles Acceleration=true
    Write Particles Position=true
   )
   &DynamicMaterialParameters(
    Index 0 Param 1=&RandomRangeFloat(Minimum=0, Maximum=1)
    Module.Index0Param0Write=true
    Module.Index0Param1Write=true
    Module.Index0Param2Write=true
    Module.Index0Param3Write=true
   )
   &ScaleColor(
    Linear Color Curve=DataInterface() {propertyValues {
      AlphaCurve {keys [<at 0.5585 1>,<at 1 0>]}
      BlueCurve {keys [<at 0 20.4998>,<at 0.17 1>]}
      GreenCurve {keys [<at 0 0>,<at 0.17 1>]}
      RedCurve {keys [<at 0 50>,<at 0.17 1>]}
     }}
    Scale Mode=::NewEnumerator2
   )
  ]
  rendererClasses [&NiagaraSpriteRendererProperties]
  renderers [{rendererClass=&NiagaraSpriteRendererProperties}]
 }
 Ash001 {
  simTarget=GPUComputeSim
  emitterUpdateScript [
   &EmitterState()
   &SpawnRate(SpawnRate=120, Use Spawn Probability=false)
  ]
  particleSpawnScript [&InitializeParticle(
    Color=<rgba 0.3 0.3 0.3 1>
    Lifetime=3
    Mesh Scale Mode=::NewEnumerator3
    Mesh Uniform Scale=500
    Sprite Rotation Mode=::NewEnumerator2
    Sprite Size Mode=::NewEnumerator1
    Sprite UV Mode=::NewEnumerator3
    Uniform Sprite Size Max=8500
    Uniform Sprite Size Min=6500
    UsePositionOffset=false
    Write Material Random=true
   )]
  particleUpdateScript [
   &ParticleState(Kill Particles When Lifetime Has Elapsed=false)
   &SkeletalMeshLocation(
    "Inherit Velocity (Scale)"=0.1
    MaskSpawnGroup=false
    Mesh Sampling Type=::NewEnumerator7
    Offset Position Along Sampled Normal=50
    OffsetAlongNormal=true
    OverrideRandomSeed=false
    Skeletal Mesh=$System.Skeletal Mesh
    Velocity Sampling=::NewEnumerator0
    VelocitySpeedLimit=false
   )
   &ScaleSpriteSize(Uniform Curve Sprite Scale=DataInterface() {propertyValues {Curve {keys [{interpMode=RCIM_Cubic, leaveTangent=-0.2205, tangentMode=RCTM_Break, time=0, value=0}, {tangentMode=RCTM_User, time=1, value=2.275}]}}}) {module=ScaleSpriteSize001}
   &KillParticles(Kill Particles=&SetBoolByFloatComparison(
     A=&ClampFloat(Float=&DotProduct(
       Vector A=$Particles.SkeletalMeshLocation.SampledNormal
       Vector B=&ReturnCameraVector(Camera Vectors=::NewEnumerator1, Coordinate Space=::World)
      ))
     B=0.258
     Comparison Type=::NewEnumerator6
    ))
   &KillParticlesInVolume(Sphere Radius=9500) {module=KillParticlesInVolume001}
   &ScaleSpriteSize(Uniform Curve Sprite Scale=DataInterface() {propertyValues {Curve {keys [
        {interpMode=RCIM_Cubic, leaveTangent=0.4272, tangentMode=RCTM_Break, time=0, value=0}
        {interpMode=RCIM_Cubic, tangentMode=RCTM_Break, time=0.5, value=2.25}
       ]}}})
   &ScaleColor(
    Linear Color Curve=DataInterface() {propertyValues {AlphaCurve {keys [<at 0 1>]}, BlueCurve {keys [<at 0 0.6667>]}, GreenCurve {keys [<at 0 0>]}, RedCurve {keys [<at 0 8>]}}}
    Scale Mode=::NewEnumerator2
   )
   &DynamicMaterialParameters(
    Index 0 Param 1=500
    Index 0 Param 2=3500
    Module.Index0Param0Write=true
    Module.Index0Param1Write=true
    Module.Index0Param2Write=true
    Module.Index0Param3Write=true
   )
   &CameraOffset(Camera Offset Amount=2500)
   /SetVariables(Particles.SpriteAlignment=&RotateVectorByQuaternion(Vector To Rotate=<1 0 0>, Quaternion=$Particles.SkeletalMeshLocation.SampledRotation)) {enabled=false, isSetParameters=true}
  ]
  rendererClasses [&NiagaraMeshRendererProperties, &NiagaraSpriteRendererProperties]
  renderers [
   {rendererClass=&NiagaraMeshRendererProperties, rendererIndex=0}
   {rendererClass=&NiagaraSpriteRendererProperties, rendererIndex=1}
  ]
 }
 Head {
  emitterUpdateScript [
   &EmitterState(Life Cycle Mode=::NewEnumerator1, Loop Duration=0.5, UseLoopDelay=false)
   &SpawnRate(SpawnRate=15, Use Spawn Probability=false)
  ]
  particleSpawnScript [
   &InitializeParticle(
    Lifetime=3
    Sprite Rotation Mode=::NewEnumerator2
    Sprite Size Mode=::NewEnumerator1
    Sprite UV Mode=::NewEnumerator3
    Uniform Sprite Size Max=1000
    Uniform Sprite Size Min=500
    UsePositionOffset=false
    Write Material Random=true
   )
   &SkeletalMeshLocation(
    "Inherit Velocity (Scale)"=0.1
    MaskSpawnGroup=false
    OverrideRandomSeed=false
    Skeletal Mesh=$System.Skeletal Mesh
    Velocity Sampling=::NewEnumerator0
    VelocitySpeedLimit=false
   )
   &ShapeLocation(MaskSpawnGroup=false, Override Seed=false, Sphere Radius=5, UseOwnerScale=true)
   &KillParticlesInVolume(Sphere Radius=9500)
   /SetVariables(Particles.SpriteFacing=&RotateVectorByQuaternion(Vector To Rotate=<1 0 0>, Quaternion=$Particles.SkeletalMeshLocation.SampledRotation)) {isSetParameters=true, module=SetVariables_#3}
  ]
  particleUpdateScript [
   &ParticleState()
   &SpriteRotationRate(Rotation Rate=10)
   &FN_HermiteSpline(
    Compute Acceleration=true
    Direction End:Vector3f=<0 -75000 15000>
    Position Blend:NiagaraFloat=$Particles.NormalizedAge
    Position End:Vector3f=<-57600 42904 10500>
    Position Start:Vector3f=$Particles.SkeletalMeshLocation.SampledPosition
    Write Particles Acceleration=true
    Write Particles Position=true
   )
   &SolveForcesAndVelocity(Clamp Velocity=false, Limit Acceleration=false)
   &DynamicMaterialParameters(
    Index 0 Param 1=&RandomRangeFloat(Minimum=0, Maximum=1)
    Module.Index0Param0Write=true
    Module.Index0Param1Write=true
    Module.Index0Param2Write=true
    Module.Index0Param3Write=true
   )
   &ScaleColor(
    Linear Color Curve=DataInterface() {propertyValues {
      AlphaCurve {keys [<at 0.5585 1>]}
      BlueCurve {keys [<at 0 20.4998>,<at 0.1023 0>]}
      GreenCurve {keys [<at 0 0>,<at 0.1023 3.1053>]}
      RedCurve {keys [<at 0 50>,<at 0.1023 5>]}
     }}
    Scale Mode=::NewEnumerator2
   )
  ]
  rendererClasses [&NiagaraSpriteRendererProperties]
  renderers [{rendererClass=&NiagaraSpriteRendererProperties}]
 }
 Trails {
  emitterUpdateScript [
   &EmitterState()
   &SpawnParticlesFromOtherEmitter(
    Attribute Reader=DataInterface() {propertyValues {EmitterBinding {bindingMode=Other, emitterName=Head}}}
    Spawn Rate=20
    Use Maximum Age=false
    Use Minimum Age=false
   )
  ]
  particleSpawnScript [
   &InitializeParticle(
    Color Mode=::NewEnumerator0
    Lifetime=0.5
    Ribbon Width=250
    Ribbon Width Mode=::NewEnumerator1
    UsePositionOffset=false
    Write Material Random=true
   )
   &SampleParticlesFromOtherEmitter(Color Sampling=::NewEnumerator0, Particle ID Sampling=::NewEnumerator1, Reject Particle on Sampled Particle's First Frame=true)
  ]
  particleUpdateScript [
   &CurlNoiseForce(Bias Noise=false, Noise Frequency=20, Noise Strength=500, Pan Noise=false)
   &SolveForcesAndVelocity(Clamp Velocity=false, Limit Acceleration=false)
   &ParticleState()
   &ScaleRibbonWidth(Ribbon Width Scale=&FloatFromCurve(FloatCurve=DataInterface() {propertyValues {Curve {keys [
         {interpMode=RCIM_Cubic, leaveTangent=4.5155, tangentMode=RCTM_Break, time=0, value=0.4}
         {interpMode=RCIM_Cubic, leaveTangent=4.5155, tangentMode=RCTM_Break, time=1, value=2.275}
        ]}}}))
  ]
  rendererClasses [&NiagaraRibbonRendererProperties]
  renderers [{rendererClass=&NiagaraRibbonRendererProperties}]
 }
}
summary {
 systemName=NS_DemonAsh_Mask
 emitters [
  {emitterName=Ash, simTarget=GPUComputeSim, rendererClasses [&NiagaraSpriteRendererProperties]}
  {
   emitterName=Ash001
   simTarget=GPUComputeSim
   rendererClasses [&NiagaraMeshRendererProperties, &NiagaraSpriteRendererProperties]
  }
  {emitterName=Head, rendererClasses [&NiagaraSpriteRendererProperties]}
  {emitterName=Trails, rendererClasses [&NiagaraRibbonRendererProperties]}
 ]
 userVariables [
  User.Skeletal Mesh {
   type=NiagaraDataInterfaceSkeletalMesh
   defaultValue {
    struct=NiagaraExt_VariableValue_DataInterface
    value {dataInterface=&self:NiagaraDataInterfaceSkeletalMesh_0}
   }
  }
  User.Spline {
   type=NiagaraDataInterfaceSpline
   description=""
   defaultValue {
    struct=NiagaraExt_VariableValue_DataInterface
    value {
     dataInterface=&self:NiagaraDataInterfaceSpline_0
     dataInterfaceClass=/Script/Niagara.NiagaraDataInterfaceSpline
    }
   }
  }
 ]
}
